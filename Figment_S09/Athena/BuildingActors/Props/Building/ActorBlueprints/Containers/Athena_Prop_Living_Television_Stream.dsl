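; Television prop Blueprint that streams a video on clients while something is within its proximity zone.
; Inner sphere overlap opens the media source and outer sphere exit closes it, giving hysteresis, and reports time spent in the zone to analytics.
; A gate blocks playback for a command-line switch, legacy media source, cellular connection or low memory, reporting memory shortfalls on entry.
; Media failure or close unbinds handlers and sends streaming analytics.
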(event EventBeginPlay)

(event OnComponentBeginOverlap(Inner_Sphere_Collision)
  (params (OverlappedComponent :type Primitive%20Component%20Object%20Reference) (OtherActor :type Actor%20Object%20Reference) (OtherComp :type Primitive%20Component%20Object%20Reference) (OtherBodyIndex :type Integer) (bFromSweep :type Boolean) (SweepResult :type Hit%20Result%20Structure%20%28by%20ref%29))
  (bind _streamingsynctest (Variables|Default|GetStreamingSyncTest))
  (if (not (Networking|IsServer))
    (if (not (Variables|Default|GetEnteredStreamingZone))
      (Variables|Default|SetEnteredStreamingZone true)
      (Variables|Default|SetTimeEnteredStreamingZone (Math|DateTime|UTCNow))
      (bind _turntelevisionon (CallFunction|ShouldTurnTelevisionOn true))
      (if _turntelevisionon
        (Media|MediaPlayer|OpenSource _streamingsynctest "/Figment_S09/Movies/StreamingSyncTestStreamedRestricted.StreamingSyncTestStreamedRestricted")
        (Media|MediaPlayer|BindEventtoOnMediaOpenFailed _streamingsynctest (delegate Custom|On%20Media%20Open%20Failed))
        (Media|MediaPlayer|BindEventtoOnMediaClosed _streamingsynctest (delegate Custom|On%20Media%20Closed))))))

(event Custom|On%20Media%20Open%20Failed
  (params (FailedUrl :type String))
  (CallFunction|TelevisionOff))

(event Custom|On%20Media%20Closed
  (CallFunction|TelevisionOff))

(event OnComponentEndOverlap(Outer_Sphere_Collision)
  (params (OverlappedComponent :type Primitive%20Component%20Object%20Reference) (OtherActor :type Actor%20Object%20Reference) (OtherComp :type Primitive%20Component%20Object%20Reference) (OtherBodyIndex :type Integer))
  (if (not (Networking|IsServer))
    (if (Variables|Default|GetEnteredStreamingZone)
      (Variables|Default|SetEnteredStreamingZone false)
      (CallFunction|SendRelevancyAnalytics)
      (bind _turntelevisionon (CallFunction|ShouldTurnTelevisionOn false))
      (if _turntelevisionon
        (Media|MediaPlayer|Close (Variables|Default|GetStreamingSyncTest) :!class /Script/MediaAssets.MediaPlayer)))))

(fn SendRelevancyAnalytics
  (for _array_element (Fortnite|GetLocalFortPlayerControllers true)
    (Analytics|FireEventStreamingMediaRelevancy _array_element (Math|Timespan|GetTotalSeconds (Math|DateTime|DateTime-DateTime (Math|DateTime|UTCNow) (Variables|Default|GetTimeEnteredStreamingZone))))))

(fn ShouldTurnTelevisionOn
  (params (ReportInsufficientMemory :type Boolean))
  ;[1 1 1] Check to see if television should never be turned on
  (bind _returnvalue (Utilities|GetCommandLine))
  (if (Math|Boolean|ORBoolean (Utilities|String|Contains _returnvalue "TelevisionOff") (FortGlobals|GetUseLegacyMediaSource) (Fortnite|IsCellConnection))
    (return false)
    (elif (>= (Fortnite|GetPhysicalMemoryMB) (FortGlobals|GetMemoryRequirementforMediaStreamingMB))
      (return true)
      (elif ReportInsufficientMemory
        (for _array_element (Fortnite|GetLocalFortPlayerControllers true)
          (Movie|ReportVideoStreamingError _array_element 0 false "Insufficient memory")
          (Movie|SendStreamingAnalytics _array_element 0))
        (return false)
        (else
          _)))))

(fn TelevisionOff
  (bind _streamingsynctest (Variables|Default|GetStreamingSyncTest))
  (Media|MediaPlayer|UnbindEventfromOnMediaOpenFailed _streamingsynctest (EventDispatchers|CreateEvent :!event On%20Media%20Open%20Failed))
  (Media|MediaPlayer|UnbindEventfromOnMediaClosed _streamingsynctest (EventDispatchers|CreateEvent :!event On%20Media%20Closed))
  (for _array_element (Fortnite|GetLocalFortPlayerControllers true)
    (Movie|SendStreamingAnalytics _array_element 0)))

(fn ConstructionScript)
